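; Blueprint for the GlowDonut water cooler prop, which exposes its static mesh components to a dynamic destruction setup.
; GetDynamicDestructionStaticMeshComponents returns three arrays: the main static mesh component alone, the water cooler tank mesh alone, and both together.
; The construction script only calls its parent, and the BeginPlay, overlap and Tick events are empty.
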(event EventBeginPlay)

(event Collision|EventActorBeginOverlap
  (params (OtherActor :type Actor%20Object%20Reference)))

(event EventTick
  (params (DeltaSeconds :type Float%20%28single-precision%29)))

(fn GetDynamicDestructionStaticMeshComponents
  (return (Utilities|Array|MakeArray (Variables|Debug|GetStaticMeshComponent)) (Utilities|Array|MakeArray (Variables|Default|GetSM_GlowDonut_WaterCoolerTank_A)) (Utilities|Array|MakeArray (Variables|Debug|GetStaticMeshComponent) (Variables|Default|GetSM_GlowDonut_WaterCoolerTank_A))))

(fn ConstructionScript
  (|CallParentFunction))
